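; Dual revolvers weapon actor with a timed Hangtime state affecting spread, feel and cosmetics.
; The replicated Hangtime struct's OnRep notifies listeners, swaps muzzle Niagara, sets audio burst rate and runs a countdown timer.
; Gameplay effects reduce spread by rarity and can add per-shot spread, with hotfix ScalableFloats for tuning.
; Handlers override Cascade impacts, pick reload rumble, multicast looping cues and clear camera modifiers and timers on unequip.
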(event EventEndPlay
  (params (EndPlayReason :type EEndPlayReason%20Enum :default "Destroyed"))
  ;[0.974 0.4069 0.2479 0.5] Clean up on End
  (|CallParentFunction EndPlayReason)
  (CallFunction|DisableAnyCurrentRevolverCameraModifiers true true)
  (CallFunction|ClearHangtimeTimerHandles))

(event EventBeginPlay
  (|CallParentFunction)
  (CallFunction|SetHangtimeTotalTime))

(event FortWeapon|EventOnEquippedAndAttached
  (|CallParentFunction)
  (bind _asfort_player_pawn (Utilities|Casting|CastToFortPlayerPawn (Game|GetInstigator)))
  (Variables|Default|SetAs%20Fort%20Player%20Pawn _asfort_player_pawn)
  (CallFunction|CacheHotfixVars)
  (EventDispatchers|BindEventtoOnReloadStarted (EventDispatchers|CreateEvent :!event Custom|OnReload))
  (EventDispatchers|BindEventtoOnPlayImpactFXDelegate (EventDispatchers|CreateEvent :!event Custom|HangtimeOverrideHitImpacts))
  ;[0.8802 0.01189 0.637 0.5] Play equip montages
  (Animation|Montage|PlayMontage (CallFunction|GetLeftHandWeaponMesh) "/CherrySmokeDualRevolversGameplay/Assets/Anim/WeaponAnim/CherrySmoke_DualRevolver_Equip_L_WEP_M.CherrySmoke_DualRevolver_Equip_L_WEP_M")
  (Animation|Montage|PlayMontage (Visual|GetWeaponMesh) "/CherrySmokeDualRevolversGameplay/Assets/Anim/WeaponAnim/CherrySmoke_DualRevolver_Equip_R_WEP_M.CherrySmoke_DualRevolver_Equip_R_WEP_M")
  (Utilities|FlowControl|Sequence
    (:then_0
      (Variables|Hangtime|GetEnableHangtimeOnFiredSpread
        (:then
          (RangedWeaponDual|BindEventtoOnHandFired (EventDispatchers|CreateEvent :!event HangtimeModifySpreadonShot)))
        (:else)))))

(event FortWeapon|EventOnUnEquippedAndDetached
  (|CallParentFunction)
  (CallFunction|DisableAnyCurrentRevolverCameraModifiers true true)
  (EventDispatchers|UnbindEventfromOnReloadStarted (EventDispatchers|CreateEvent :!event Custom|OnReload))
  (CallFunction|StopClientReloadForceFeedbackonWeaponSwap)
  (CallFunction|ClearHangtimeTimerHandles)
  (Utilities|FlowControl|Sequence
    (:then_0
      (Variables|Hangtime|GetEnableHangtimeOnFiredSpread
        (:then)
        (:else
          (RangedWeaponDual|UnbindEventfromOnHandFired (EventDispatchers|CreateEvent :!event HangtimeModifySpreadonShot)))))
    (:then_1
      (Networking|SwitchHasAuthority
        (:Authority
          (CallFunction|ActivateNSForHangtime false))))))

(event Custom|SetHangtimeAbilityActive
  (params (HangtimeStruct :type Hangtime%20Struct%20Structure))
  ;[0 0.2486 0.9115 0.5] Set UI call for Hangtime, and update muzzle FX
  (Variables|Hangtime|SetHangtimeStruct HangtimeStruct)
  ;[0.4655 0.8388 0.849 0.5] Reset Fall Height
  (if (Variables|Hangtime|GetHangtimeResetFallHeight)
    (Pawn|Character|ResetFallingHeight (Variables|Default|GetAs%20Fort%20Player%20Pawn))))

(event Custom|HangtimeModifySpreadOnShot
  (params (Hand :type EDualWeaponHand%20Enum :default "LEFT") (bPersistantFire :type Boolean))
  ;[0.1352 0 0.9115 0.5] Hangtime Spread modifer on shot. This will add stacking spread that will fall off over time.
  ; Intention is to allow a balance dial when in hangtime in case we don't want perfect accuracy with repeating firing
  (GameplayEffects|ApplyGameplayEffectToSelf :self (Ability|GetAbilitySystemComponent (Variables|Default|GetAs%20Fort%20Player%20Pawn)) :GameplayEffectClass (Variables|Hangtime|GetOn%20Fired%20Spread%20Gameplay%20Effect%20Class) :Level 1.0))

(event Custom|HangtimeSpreadReduction
  (params (bUseSpreadReduction :type Boolean))
  ;[0 0.1992 0.9271 0.5] SPREAD - add or remove spread reduction for entering and exiting hangtime
  (if bUseSpreadReduction
    (GameplayEffects|ApplyGameplayEffectToSelf :self (Ability|GetAbilitySystemComponent (Variables|Default|GetAs%20Fort%20Player%20Pawn)) :GameplayEffectClass (select (Utilities|Enum|Equal(Enum) (Rarity|GetWeaponDataRarity self) "Mythic") (Variables|Hangtime|GetSpread%20and%20Recoil%20Reduction%20Gameplay%20Effect%20Mythic%20Class) (Variables|Hangtime|GetSpread%20and%20Recoil%20Reduction%20Gameplay%20Effect%20Class)) :Level 1.0)
    (else
      (GameplayEffects|RemoveActiveEffectswithTags (Ability|GetAbilitySystemComponent (Variables|Default|GetAs%20Fort%20Player%20Pawn)) (Variables|Hangtime|GetEffectsWithTagsToRemove)))))

(event Custom|OnReload
  (params (ReloadTime :type Float%20%28single-precision%29) (ReloadType :type EFortWeaponReloadType%20Enum :default "ReloadWholeClip"))
  (bind _hangtimestruct_hangtimeactive_1_06c9fc1a43453478812ca88c8b8c9540 (Utilities|Struct|BreakHangtimeStruct (Variables|GetHangtimeStruct_HangtimeActive_1_06C9FC1A43453478812CA88C8B8C9540)))
  ;[0.000635 0.8906 0.7176 0.5] Custom Force feedback on reload based on hangtime ( reload is faster in hangtime )
  (Variables|Hangtime|SetHangtime%20reload%20time ReloadTime)
  (Game|Feedback|ClientPlayForceFeedback (Pawn|GetLocalViewingPlayerController (Variables|Default|GetAs%20Fort%20Player%20Pawn)) (select _hangtimestruct_hangtimeactive_1_06c9fc1a43453478812ca88c8b8c9540 (Variables|Default|GetHangtime%20Reload%20Force%20Feedback%20Effect) (Variables|Default|GetReload%20Force%20Feedback%20Effect)) "None" false false false))

(event Custom|HangtimeGC
  ;[0.8906 0 0.7368 0.5] Hangtime Cosmetics
  (GameplayCue|ExecuteGameplayCueOnActor(Burst) (Variables|Default|GetAs%20Fort%20Player%20Pawn) (Variables|GameplayCue|GetHangtimeStart%20Gameplay%20Cue%20Tag) (Ability|GameplayCue|MakeGameplayCueParameters 0.0)))

(event Custom|StartTrackingHangtimeTimeRemaining
  ;[0.8906 0 0.7368 0.5] Hangtime Cosmetics
  ;;[0.8698 0 0.6599 0.5] Track time remaining for FX
  (bind _returnvalue (Utilities|Time|SetTimerbyEvent (EventDispatchers|CreateEvent :!event Custom|TrackHangtimeRemaining) (Variables|Hangtime|GetHangtimeTimeTrackingTimeSlice) true true))
  (Variables|Hangtime|SetHangtimeTimeRemainingTimerHandle _returnvalue))

(event Custom|TrackHangtimeRemaining
  (bind _hangtimetimetrackingtimeslice (Variables|Hangtime|GetHangtimeTimeTrackingTimeSlice))
  (bind _hangtimetimeremaining (Variables|Hangtime|GetHangtimeTimeRemaining))
  (bind _returnvalue (- _hangtimetimeremaining _hangtimetimetrackingtimeslice))
  ;[0.8906 0 0.7368 0.5] Hangtime Cosmetics
  ;;[0.8698 0 0.6599 0.5] Track time remaining for FX
  (if (<= _returnvalue 0)
    (CallFunction|ClearHangtimeTimerHandles)
    (else
      (Variables|Hangtime|SetHangtimeTimeRemaining _returnvalue)
      (Niagara|SetNiagaraVariable(Float) (Variables|Default|GetNiagaraLeftHand_TEMP1) "HangtimeTimeRemaining" 1.0))))

(event Custom|HangtimeOverrideHitImpacts
  (params (Weapon :type Fort%20Weapon%20Object%20Reference) (HitResult :type Hit%20Result%20Structure%20%28by%20ref%29) (FXSystemComponent :type FXSystem%20Component%20Object%20Reference))
  (bind (_bblockinghit _binitialoverlap _time _distance _location _impactpoint _normal _impactnormal _physmat) (Collision|BreakHitResult HitResult))
  (bind _hangtimestruct_hangtimeactive_1_06c9fc1a43453478812ca88c8b8c9540 (Utilities|Struct|BreakHangtimeStruct (Variables|GetHangtimeStruct_HangtimeActive_1_06C9FC1A43453478812CA88C8B8C9540)))
  ;[0 0.2486 0.9115 0.5] HANGTIME BULLET IMPACT PARTICLE OVERRIDE
  ; NOTE: Not sure if there's a similar way to do this in NS... using cascade here isn't ideal, but we can override it here
  (if (and _hangtimestruct_hangtimeactive_1_06c9fc1a43453478812ca88c8b8c9540 (Utilities|Enum|NotEqual(Enum) (Variables|/Script/PhysicsCore.PhysicalMaterial|GetSurfaceType _physmat) "SurfaceType8"))
    (bind _ascascade_particle_system_component (Utilities|Casting|CastToParticleSystemComponent FXSystemComponent))
    (Effects|Components|ParticleSystem|SetTemplate _ascascade_particle_system_component (Variables|Default|GetHangtime%20Impact%20Particle%20Template))))

(event Custom|ActivateNSForHangtime
  (params (Activate :type Boolean))
  ;[0.8906 0 0.7368 0.5] Hangtime Cosmetics
  (CallFunction|MulktiCastActivateNSForHangtime Activate))

(event Custom|MulktiCast_ActivateNSForHangtime
  (params (Activate :type Boolean))
  (bind _niagaralefthand_temp1 (Variables|Default|GetNiagaraLeftHand_TEMP1))
  ;[0.8906 0 0.7368 0.5] Hangtime Cosmetics
  (if Activate
    (Components|Activation|Activate _niagaralefthand_temp1 false)
    (Transformation|AttachComponentToComponent (Variables|Default|GetNiagaraRightHand_TEMP) (Visual|GetWeaponMesh) "handle" "KeepRelative" "KeepRelative" "KeepRelative" true)
    (Transformation|AttachComponentToComponent (Variables|Default|GetNiagaraLeftHand_TEMP1) (CallFunction|GetLeftHandWeaponMesh) "handle" "KeepRelative" "KeepRelative" "KeepRelative" true)
    (GameplayCue|AddGameplayCueOnActor(Looping) (Variables|Default|GetAs%20Fort%20Player%20Pawn) (Variables|GameplayCue|GetHangtime%20Looping%20Gameplay%20Cue%20Tag) (Ability|GameplayCue|MakeGameplayCueParameters 0.0))
    (else
      (Components|Activation|Deactivate _niagaralefthand_temp1)
      (GameplayCue|RemoveGameplayCueOnActor(Looping) (Variables|Default|GetAs%20Fort%20Player%20Pawn) (Variables|GameplayCue|GetHangtime%20Looping%20Gameplay%20Cue%20Tag) (Ability|GameplayCue|MakeGameplayCueParameters 0.0)))))

(fn CacheHotfixVars
  (Variables|Hangtime|SetHangtimeResetFallHeight (ScalableFloat|GetValueasBool (Variables|Hotfix|GetHF_HangtimeResetFallHeight)))
  (Variables|Hangtime|SetEnableHangtimeOnFiredSpread (ScalableFloat|GetValueasBool (Variables|Hotfix|GetHF_EnableHangtimeOnFiredSpread))))

(fn ClearHangtimeTimerHandles
  (Utilities|Time|ClearandInvalidateTimerbyHandle (Variables|Hangtime|GetHangtimeTimeRemainingTimerHandle)))

(fn DisableAnyCurrentRevolverCameraModifiers
  (params (Immediate :type Boolean) (DisableVignette :type Boolean))
  (bind _as_fort_player_pawn (Variables|Default|GetAs%20Fort%20Player%20Pawn))
  (bind _outmodifier (Class|WeaponFunctionLibrary|FindorAddCameraModifierforPawn (Variables|Camera|GetDodge%20Camera%20Modifier%20Class) _as_fort_player_pawn))
  (CameraModifier|DisableModifier _outmodifier Immediate)
  (bind _outmodifier_2 (Class|WeaponFunctionLibrary|FindorAddCameraModifierforPawn (Variables|Camera|GetDodge%20Right%20Camera%20Modifier%20Class) _as_fort_player_pawn))
  (CameraModifier|DisableModifier _outmodifier_2 Immediate)
  (if DisableVignette
    (bind _outmodifier_3 (Class|WeaponFunctionLibrary|FindorAddCameraModifierforPawn (Variables|Camera|GetVignette%20Camera%20Modifier%20Class) _as_fort_player_pawn))
    (bind _ascherry_smoke_dual_revolvers_camera_modifier_vignette (Utilities|Casting|CastToCherrySmokeDualRevolvers_CameraModifier_Vignette _outmodifier_3))
    (Variables|/CherrySmokeDualRevolversGameplay/Gameplay/Camera/CherrySmokeDualRevolvers_CameraModifier_Vignette.CherrySmokeDualRevolvers_CameraModifier_Vignette_C|SetIsActive false _ascherry_smoke_dual_revolvers_camera_modifier_vignette)))

(fn GetAnimData
  (return (Utilities|Struct|BreakHangtimeStruct (Variables|Hangtime|GetHangtimeStruct)) (Variables|Hangtime|GetHangtime%20reload%20time)))

(fn OnRep_CanDoubleJump)

(fn OnRep_HangtimeStruct
  (bind _returnvalue (AudioParameter|GetAudioParameterComponent))
  (bind (_hangtimestruct_hangtimeactive_1_06c9fc1a43453478812ca88c8b8c9540 _hangtimestruct_timeremaining_4_2753e25d431ae94a12030e9b70d7c35a) (Utilities|Struct|BreakHangtimeStruct (Variables|GetHangtimeStruct_HangtimeActive_1_06C9FC1A43453478812CA88C8B8C9540)))
  (bind (_hangtimestruct_hangtimeactive_1_06c9fc1a43453478812ca88c8b8c9540_2 _hangtimestruct_timeremaining_4_2753e25d431ae94a12030e9b70d7c35a_2) (Utilities|Struct|BreakHangtimeStruct (Variables|GetHangtimeStruct_HangtimeActive_1_06C9FC1A43453478812CA88C8B8C9540)))
  (Default|CallHangtimeActiveDispatcher _hangtimestruct_hangtimeactive_1_06c9fc1a43453478812ca88c8b8c9540 _hangtimestruct_timeremaining_4_2753e25d431ae94a12030e9b70d7c35a)
  (if _hangtimestruct_hangtimeactive_1_06c9fc1a43453478812ca88c8b8c9540
    (Niagara|SetNiagaraSystemAsset (Variables|Default|GetMuzzle%28Empty%29%20Niagara) (Variables|Default|GetHangtimeMuzzleNS) true)
    (Variables|Hangtime|SetHangtimeTimeRemaining _hangtimestruct_timeremaining_4_2753e25d431ae94a12030e9b70d7c35a_2)
    (CallFunction|StartTrackingHangtimeTimeRemaining)
    (Networking|CanExecuteCosmeticEvents
      (:True
        (Audio|Parameter|SetFloatParameter _returnvalue "BurstRate" 0.25 :!class /Script/AudioExtensions.AudioParameterControllerInterface)))
    (else
      (Niagara|SetNiagaraSystemAsset (Variables|Default|GetLeftMuzzle%28Empty%29%20Niagara) (Variables|Default|GetMuzzleNS) true)
      (CallFunction|ClearHangtimeTimerHandles)
      (Networking|CanExecuteCosmeticEvents
        (:True
          (Audio|Parameter|SetFloatParameter _returnvalue "BurstRate" 0.4 :!class /Script/AudioExtensions.AudioParameterControllerInterface))))))

(fn SetHangtimeTotalTime
  (bind _output_get (Variables|Hangtime|SetHangtimeTotalTime (ScalableFloat|GetValueatLevel (Variables|Hotfix|GetHF_HangtimeTotalTime))))
  (Variables|SetHangtimeStruct_HangtimeActive_1_06C9FC1A43453478812CA88C8B8C9540 (Utilities|Struct|MakeHangtimeStruct false _output_get)))

(fn StopClientReloadForceFeedbackOnWeaponSwap
  (bind _hangtimestruct_hangtimeactive_1_06c9fc1a43453478812ca88c8b8c9540 (Utilities|Struct|BreakHangtimeStruct (Variables|GetHangtimeStruct_HangtimeActive_1_06C9FC1A43453478812CA88C8B8C9540)))
  (Game|Feedback|ClientStopForceFeedback (Pawn|GetLocalViewingPlayerController (Variables|Default|GetAs%20Fort%20Player%20Pawn)) (select _hangtimestruct_hangtimeactive_1_06c9fc1a43453478812ca88c8b8c9540 (Variables|Default|GetHangtime%20Reload%20Force%20Feedback%20Effect) (Variables|Default|GetReload%20Force%20Feedback%20Effect)) "None"))

(fn ConstructionScript
  (|CallParentFunction))
